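# Post-process material that computes an anamorphic lens undistortion displacement map for camera calibration, using the 3DE4 Anamorphic Standard Degree 4 model.
# One Custom HLSL node takes UVs and scalar parameters for filmback size, pixel aspect, coefficients, squeeze and mounting angle.
# It converts to rotated, diagonally normalized coordinates, inverts distortion by ten fixed-point iterations, then returns signed UV offsets.
# Negative emissive is allowed; lens center offset is not applied.
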Material /CameraCalibrationCore/Materials/M_AnamorphicUndistortionDisplacementMap
graph {
 properties {
  bAllowNegativeEmissiveColor=true
  blendMode=BLEND_Masked
  materialDomain=MD_PostProcess
  pixelDepthOffsetMode=PDOM_Legacy
  shadingModel=MSM_Unlit
  usedShadingModels=Unlit
 }
 nodes [
  MaterialExpressionTextureCoordinate_0 = TextureCoordinate()
  MaterialExpressionScalarParameter_5 = ScalarParameter() {defaultValue=1920, parameterName=w_fb}
  MaterialExpressionScalarParameter_6 = ScalarParameter() {defaultValue=1080, parameterName=h_fb}
  MaterialExpressionScalarParameter_1 = ScalarParameter() {defaultValue=1, parameterName=pixel_aspect}
  MaterialExpressionScalarParameter_0 = ScalarParameter() {parameterName=cx02}
  MaterialExpressionScalarParameter_7 = ScalarParameter() {parameterName=cx04}
  MaterialExpressionScalarParameter_8 = ScalarParameter() {parameterName=cx22}
  MaterialExpressionScalarParameter_9 = ScalarParameter() {parameterName=cx24}
  MaterialExpressionScalarParameter_12 = ScalarParameter() {parameterName=cx44}
  MaterialExpressionScalarParameter_13 = ScalarParameter() {parameterName=cy02}
  MaterialExpressionScalarParameter_14 = ScalarParameter() {parameterName=cy04}
  MaterialExpressionScalarParameter_15 = ScalarParameter() {parameterName=cy22}
  MaterialExpressionScalarParameter_16 = ScalarParameter() {parameterName=cy24}
  MaterialExpressionScalarParameter_17 = ScalarParameter() {parameterName=cy44}
  MaterialExpressionScalarParameter_3 = ScalarParameter() {defaultValue=1, parameterName=sx}
  MaterialExpressionScalarParameter_4 = ScalarParameter() {defaultValue=1, parameterName=sy}
  MaterialExpressionScalarParameter_2 = ScalarParameter() {parameterName=phi_mnt}
  Custom!(
   input_uv=@MaterialExpressionTextureCoordinate_0
   w_fb=@MaterialExpressionScalarParameter_5
   h_fb=@MaterialExpressionScalarParameter_6
   pixel_aspect=@MaterialExpressionScalarParameter_1
   cx02=@MaterialExpressionScalarParameter_0
   cx04=@MaterialExpressionScalarParameter_7
   cx22=@MaterialExpressionScalarParameter_8
   cx24=@MaterialExpressionScalarParameter_9
   cx44=@MaterialExpressionScalarParameter_12
   cy02=@MaterialExpressionScalarParameter_13
   cy04=@MaterialExpressionScalarParameter_14
   cy22=@MaterialExpressionScalarParameter_15
   cy24=@MaterialExpressionScalarParameter_16
   cy44=@MaterialExpressionScalarParameter_17
   sx=@MaterialExpressionScalarParameter_3
   sy=@MaterialExpressionScalarParameter_4
   phi_mnt=@MaterialExpressionScalarParameter_2
  ) {
   code="// The following implementation is based on the 3DE4 Anamorphic - Standard, Degree 4 model for distortion\r\n\r\n##define UNDISTORT_ITERATION_COUNT 10\r\n\r\n// Pre-compute some values that will be useful later\r\nfloat r_fb = (0.5) * sqrt(w_fb*w_fb + h_fb*h_fb);\r\n\r\nfloat phi_mnt_rad = phi_mnt * (PI / 180);\r\nfloat cos_phi_mnt = cos(phi_mnt_rad);\r\nfloat sin_phi_mnt = sin(phi_mnt_rad);\r\n\r\n// Step 1: Transform input UVs into diagonally-normalized coordinates, based on the width and height of the filmback/image\r\nfloat x_dn = (input_uv.x - 0.5) * (w_fb / r_fb); // don't forget to factor in lens center offset later\r\nfloat y_dn = (input_uv.y - 0.5) * (h_fb / r_fb); // don't forget to factor in lens center offset later\r\n\r\n// Step 2: Rotate x and y by the inverse of the mounting angle\r\nfloat x_dn_rotated = (x_dn * cos_phi_mnt) + (y_dn * sin_phi_mnt);\r\nfloat y_dn_rotated = (y_dn * cos_phi_mnt) - (x_dn * sin_phi_mnt);\r\n\r\n// Step 3: Scale the x-coordinate by the inverse of the pixel aspect ratio\r\nx_dn_rotated = x_dn_rotated / pixel_aspect;\r\n\r\nfloat x_dn_prime = x_dn_rotated;\r\nfloat y_dn_prime = y_dn_rotated;\r\n\r\n// Step 4: Compute the anamorphic distortion for x and y, to the 4th degree\r\nfor(int i = 0; i < UNDISTORT_ITERATION_COUNT; i++)\r\n{\r\n    float r_dn_sq = x_dn_prime*x_dn_prime + y_dn_prime*y_dn_prime;\r\n    float r_dn_4th = r_dn_sq * r_dn_sq;\r\n    float phi = atan2(y_dn_prime, x_dn_prime);\r\n\r\n    float distx = (1 + (cx02 * r_dn_sq) + (cx04 * r_dn_4th) + (cx22 * r_dn_sq * cos(2 * phi)) + (cx24 * r_dn_4th * cos(2 * phi)) + (cx44 * r_dn_4th * cos(4 * phi)));\r\n    float disty = (1 + (cy02 * r_dn_sq) + (cy04 * r_dn_4th) + (cy22 * r_dn_sq * cos(2 * phi)) + (cy24 * r_dn_4th * cos(2 * phi)) + (cy44 * r_dn_4th * cos(4 * phi)));\r\n\r\n    x_dn_prime = (x_dn_rotated / distx) / sx;\r\n    y_dn_prime = (y_dn_rotated / disty) / sy;\r\n}\r\n\r\n// Step 5: Rotate x and y by the mounting angle\r\nfloat x_dn_prime_rotated = (x_dn_prime * cos_phi_mnt) - (y_dn_prime * sin_phi_mnt);\r\nfloat y_dn_prime_rotated = (y_dn_prime * cos_phi_mnt) + (x_dn_prime * sin_phi_mnt);\r\n\r\n// Step 6: Scale the x-coordinate by the pixel aspect ratio\r\nx_dn_prime_rotated = x_dn_prime_rotated * pixel_aspect;\r\n\r\n// Step 7: Transform diagonally normalized coordinates back into unit coordinates\r\nfloat x_unit_prime = ((x_dn_prime_rotated * r_fb) / w_fb) + 0.5;\r\nfloat y_unit_prime = ((y_dn_prime_rotated * r_fb) / h_fb) + 0.5;\r\n\r\n// Step 8: Compute the offset between this point and the the input UV\r\nfloat2 output_uv;\r\noutput_uv.x = x_unit_prime - input_uv.x;\r\noutput_uv.y = y_unit_prime - input_uv.y;\r\n\r\nreturn output_uv;"
   outputType=CMOT_Float2
   inputs [
    {
     inputName=input_uv
     input {expression=&self:MaterialExpressionTextureCoordinate_0}
    }
    {
     inputName=w_fb
     input {expression=&self:MaterialExpressionScalarParameter_5}
    }
    {
     inputName=h_fb
     input {expression=&self:MaterialExpressionScalarParameter_6}
    }
    {
     inputName=pixel_aspect
     input {expression=&self:MaterialExpressionScalarParameter_1}
    }
    {
     inputName=cx02
     input {expression=&self:MaterialExpressionScalarParameter_0}
    }
    {
     inputName=cx04
     input {expression=&self:MaterialExpressionScalarParameter_7}
    }
    {
     inputName=cx22
     input {expression=&self:MaterialExpressionScalarParameter_8}
    }
    {
     inputName=cx24
     input {expression=&self:MaterialExpressionScalarParameter_9}
    }
    {
     inputName=cx44
     input {expression=&self:MaterialExpressionScalarParameter_12}
    }
    {
     inputName=cy02
     input {expression=&self:MaterialExpressionScalarParameter_13}
    }
    {
     inputName=cy04
     input {expression=&self:MaterialExpressionScalarParameter_14}
    }
    {
     inputName=cy22
     input {expression=&self:MaterialExpressionScalarParameter_15}
    }
    {
     inputName=cy24
     input {expression=&self:MaterialExpressionScalarParameter_16}
    }
    {
     inputName=cy44
     input {expression=&self:MaterialExpressionScalarParameter_17}
    }
    {
     inputName=sx
     input {expression=&self:MaterialExpressionScalarParameter_3}
    }
    {
     inputName=sy
     input {expression=&self:MaterialExpressionScalarParameter_4}
    }
    {
     inputName=phi_mnt
     input {expression=&self:MaterialExpressionScalarParameter_2}
    }
   ]
  }
 ]
}
