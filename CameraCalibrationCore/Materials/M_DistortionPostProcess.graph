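# Post-process material in CameraCalibrationCore that applies lens distortion by resampling the scene image through a displacement map.
# Viewport UVs are offset by the DistortionDisplacementMap parameter texture, with its red channel feeding both axes.
# The offset UVs are scaled about screen center by dividing by overscan_factor.
# PostProcessInput0 is sampled there with filtering and split into components, with alpha output enabled on the material.
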Material /CameraCalibrationCore/Materials/M_DistortionPostProcess
graph {
 properties {blendableOutputAlpha=true, materialDomain=MD_PostProcess, pixelDepthOffsetMode=PDOM_Legacy, usedShadingModels=Default Lit}
 nodes [
  n1 = ScreenPosition()
  n2 = TextureSample(UVs=@n1.ViewportUV, Tex=TextureObjectParameter() {parameterName=DistortionDisplacementMap})
  n3 = AppendVector(A=@n2.R, B=@n2.R)
  n4 = Divide(
   A=Subtract(A=Add(A=@n1.ViewportUV, B=@n3)) {constB=0.5}
   B=ScalarParameter() {defaultValue=1, parameterName=overscan_factor, sliderMax=1.2, sliderMin=1}
  )
  MaterialFunctionCall(Float4=SceneTexture(UVs=Add(A=@n4) {constB=0.5}) {bFiltered=true, sceneTextureId=PPI_PostProcessInput0}.Color) {materialFunction=&BreakOutFloat4Components}
 ]
}
