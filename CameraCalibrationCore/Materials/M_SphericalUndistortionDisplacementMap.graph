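# Post-process material for camera calibration that generates a spherical lens undistortion displacement map from lens parameters.
# A Custom HLSL node centers UVs on 0.5, normalizes by focal length, applies radial (k1-k3) and tangential (p1, p2) terms, then rescales and offsets by cx/cy.
# Subtracting the original UV yields a signed offset, and negative emissive color is allowed.
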Material /CameraCalibrationCore/Materials/M_SphericalUndistortionDisplacementMap
graph {
 properties {
  bAllowNegativeEmissiveColor=true
  materialDomain=MD_PostProcess
  pixelDepthOffsetMode=PDOM_Legacy
  shadingModel=MSM_Unlit
  usedShadingModels=Unlit
 }
 nodes [
  MaterialExpressionTextureCoordinate_0 = TextureCoordinate()
  MaterialExpressionScalarParameter_0 = ScalarParameter() {parameterName=k1}
  MaterialExpressionScalarParameter_1 = ScalarParameter() {parameterName=k2}
  MaterialExpressionScalarParameter_4 = ScalarParameter() {parameterName=k3}
  MaterialExpressionScalarParameter_2 = ScalarParameter() {parameterName=p1}
  MaterialExpressionScalarParameter_3 = ScalarParameter() {parameterName=p2}
  MaterialExpressionScalarParameter_10 = ScalarParameter() {defaultValue=0.5, parameterName=cx}
  MaterialExpressionScalarParameter_11 = ScalarParameter() {defaultValue=0.5, parameterName=cy}
  MaterialExpressionScalarParameter_5 = ScalarParameter() {defaultValue=1, parameterName=fx}
  MaterialExpressionScalarParameter_6 = ScalarParameter() {defaultValue=1, parameterName=fy}
  n1 = Custom!(
   uvi=@MaterialExpressionTextureCoordinate_0
   k1=@MaterialExpressionScalarParameter_0
   k2=@MaterialExpressionScalarParameter_1
   k3=@MaterialExpressionScalarParameter_4
   p1=@MaterialExpressionScalarParameter_2
   p2=@MaterialExpressionScalarParameter_3
   cx=@MaterialExpressionScalarParameter_10
   cy=@MaterialExpressionScalarParameter_11
   fx=@MaterialExpressionScalarParameter_5
   fy=@MaterialExpressionScalarParameter_6
  ) {
   code="\r\nfloat x_prime = (uvi.x - 0.5) / fx;\r\nfloat y_prime = (uvi.y - 0.5) / fy;\r\n\r\n//\r\n//\r\n// Compute undistortion\r\n//\r\n//\r\nfloat x_prime2 = x_prime*x_prime;\r\nfloat y_prime2 = y_prime*y_prime;\r\n\r\nfloat r2 = x_prime2 + y_prime2;\r\nfloat r4 = r2*r2;\r\nfloat r6 = r2*r2*r2;\r\n\r\n// radial distortion\r\nfloat dr = 1 + k1*r2 + k2*r4 + k3*r6;\r\n\r\n// tangential distortion\r\nfloat dtx = 2*p1*x_prime*y_prime + p2*(r2 + 2*x_prime2);\r\nfloat dty = p1*(r2 + 2*y_prime2) + 2*p2*x_prime*y_prime;\r\n\r\nfloat x_double_prime = (x_prime*dr) + dtx;\r\nfloat y_double_prime = (y_prime*dr) + dty;\r\n\r\n// convert to uv\r\nfloat x_undistort = (fx * x_double_prime) + cx;\r\nfloat y_undistort = (fy * y_double_prime) + cy;\r\n\r\nfloat2 result;\r\nresult.x = x_undistort;\r\nresult.y = y_undistort;\r\n\r\nreturn result;"
   outputType=CMOT_Float2
   inputs [
    {
     inputName=uvi
     input {expression=&self:MaterialExpressionTextureCoordinate_0}
    }
    {inputName=k1, input {expression=&self:MaterialExpressionScalarParameter_0}}
    {inputName=k2, input {expression=&self:MaterialExpressionScalarParameter_1}}
    {inputName=k3, input {expression=&self:MaterialExpressionScalarParameter_4}}
    {inputName=p1, input {expression=&self:MaterialExpressionScalarParameter_2}}
    {inputName=p2, input {expression=&self:MaterialExpressionScalarParameter_3}}
    {
     inputName=cx
     input {expression=&self:MaterialExpressionScalarParameter_10}
    }
    {
     inputName=cy
     input {expression=&self:MaterialExpressionScalarParameter_11}
    }
    {inputName=fx, input {expression=&self:MaterialExpressionScalarParameter_5}}
    {inputName=fy, input {expression=&self:MaterialExpressionScalarParameter_6}}
   ]
  }
  Subtract(A=@n1, B=TextureCoordinate())
 ]
}
